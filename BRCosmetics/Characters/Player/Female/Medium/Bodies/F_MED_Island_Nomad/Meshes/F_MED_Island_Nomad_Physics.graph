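# Physics asset for the Island Nomad medium female player body, simulating a dangling knife with two feathers at the right thigh.
# A small knife base capsule is locked in swing and twist to a large thigh capsule.
# The knife capsule hangs from that base with narrow swing and twist limits.
# Both feather boxes constrain to the knife with wide swing and tight twist limits.
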PhysicsAsset /BRCosmetics/Characters/Player/Female/Medium/Bodies/F_MED_Island_Nomad/Meshes/F_MED_Island_Nomad_Physics
bodies {
 dyn_feather_1 [{shapeType=Box, extentX=7.5, extentY=3, extentZ=1, shapeName=dyn_feather_1_box, center=<4.225 0.7195 -1.463>, rotation=<rot -23.2028 0 0>}]
 dyn_feather_2 [{
   shapeType=Box
   extentX=9.5
   extentY=3
   extentZ=1
   shapeName=dyn_feather_2_box
   center=<6.144 1.1071 -0.7632>
   rotation=<rot -11.4933 7.6908 10.8722>
  }]
 dyn_knife_1 [{
   shapeType=Capsule
   length=24.9942
   radius=2.0732
   shapeName=dyn_knife_1_capsule
   center=<1.8765 0.1649 0.4203>
   rotation=<rot 0 -90.4437 -84.6528>
  }]
 dyn_knife_base [{shapeType=Capsule, length=0.505, radius=0.505, shapeName=dyn_knife_base_capsule, rotation=<rot 0 -90 0>}]
 thigh_r [{
   shapeType=Capsule
   length=17.9884
   radius=16.6674
   shapeName=thigh_r_capsule
   center=<21.1179 1.9449 -9.3383>
   rotation=<rot -6.798 5.1524 -107.8249>
  }]
}
constraints [
 {bone1Name=dyn_knife_1, bone2Name=dyn_knife_base, swing1LimitDegrees=15, swing2LimitDegrees=7, twistLimitDegrees=3}
 {
  bone1Name=dyn_knife_base
  bone2Name=thigh_r
  swing1LimitDegrees=45
  swing1Motion=Locked
  swing2LimitDegrees=45
  swing2Motion=Locked
  twistLimitDegrees=45
  twistMotion=Locked
 }
 {bone1Name=dyn_feather_1, bone2Name=dyn_knife_1, swing1LimitDegrees=60, swing2LimitDegrees=60, twistLimitDegrees=5}
 {bone1Name=dyn_feather_2, bone2Name=dyn_knife_1, swing1LimitDegrees=40, swing2LimitDegrees=60, twistLimitDegrees=5}
]
