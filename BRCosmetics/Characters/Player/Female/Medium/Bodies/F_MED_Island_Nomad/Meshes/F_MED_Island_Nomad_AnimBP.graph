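# Animation Blueprint for the female medium Island Nomad body cosmetic, adding deformation and secondary motion to a pose copied from the attached parent mesh.
# Twist and sculpter Control Rigs feed a RigidBody simulation relative to the right thigh.
# Weighted constraints then adjust back-belt, knife and zipper bones, and AnimDynamics adds sprung zipper motion.
# Each update, a deprecated state function computes the RigidBody gravity override from parameters including skydive-float wind.
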AnimBlueprint /BRCosmetics/Characters/Player/Female/Medium/Bodies/F_MED_Island_Nomad/Meshes/F_MED_Island_Nomad_AnimBP
target_skeleton=&/F_MED_Island_Nomad_Skeleton
graphs [
 AnimGraph {nodes [
   n1 = Control Rig(Source=Copy Pose From Mesh() {bUseAttachedParent=true}) {
    alphaBoolBlend {blendOption=Linear}
    controlRigAssetReference {blueprintRigClass=&/F_MED_Island_Nomad_Deform_CtrlRig_twists.F_MED_Island_Nomad_Deform_CtrlRig_twists_C}
   }
   n2 = Control Rig!(
    Source=@n1
    Alpha
    KneeRot_fr_l_114deg:Vector
    KneeTrans_fr_l_114deg:Vector="0.258,-3.591,0.009"
    KneeScale_fr_l_114deg:Vector="1,1,1"
    KneeScale_fr_l_130deg:Vector="1,1,1"
    KneeRot_fr_l_130deg:Vector="0,0,-51.2"
    KneeTrans_fr_l_130deg:Vector="0.258,-3.591,0.009"
   ) -> (Pose) {
    alphaBoolBlend {blendOption=Linear}
    controlRigAssetReference {blueprintRigClass=&/F_MED_Island_Nomad_Deform_CtrlRig_sculpters.F_MED_Island_Nomad_Deform_CtrlRig_sculpters_C}
   }
   n3 = RigidBody!(ComponentPose=Local To Component(LocalPose=@n2), Alpha, OverrideWorldGravity:Vector=Get Gravity Override(), ExternalForce) -> (Pose) {
    bClampLinearTranslationLimitToRefPose=true
    bOverrideWorldGravity=true
    overridePhysicsAsset=&/F_MED_Island_Nomad_Physics
    simulationSpace=BaseBoneSpace
    alphaBoolBlend {blendOption=Linear}
    baseBoneRef {boneName=thigh_r}
    componentAppliedLinearAccClamp=<3000 3000 3000>
    componentLinearAccScale=<0.03 0.03 0.03>
    componentLinearVelScale=<0.03 0.03 0.03>
    overrideWorldGravity {z=-980}
   }
   n4 = "Constraint - dyn_beltBack_1 (PP)"!(ComponentPose=@n3, Alpha=1, ConstraintWeights_0=0.8, ConstraintWeights_1:float=0.2) -> (Pose) {
    alphaBoolBlend {blendOption=Linear}
    boneToModify {boneName=dyn_beltBack_1}
    constraintSetup [{targetBone {boneName=pelvis}}, {targetBone {boneName=deform_thigh_bk_r}}]
    constraintWeights [1,1]
   }
   n5 = "Constraint - dyn_knife_base (P)"(ComponentPose=@n4, Alpha=1, ConstraintWeights_0=1) {
    alphaBoolBlend {blendOption=Linear}
    boneToModify {boneName=dyn_knife_base}
    constraintSetup [{targetBone {boneName=thigh_twist_01_r}}]
    constraintWeights [1]
   }
   n6 = "Constraint - dyn_zipper_1 (PPPP)"!(
    ComponentPose=@n5
    Alpha=1
    ConstraintWeights_0=0.02
    ConstraintWeights_1:float=0.3
    ConstraintWeights_2:float=0.6
    ConstraintWeights_3:float=0.08
   ) -> (Pose) {
    alphaBoolBlend {blendOption=Linear}
    boneToModify {boneName=dyn_zipper_1}
    constraintSetup [
     {targetBone {boneName=thigh_twist_01_r}}
     {targetBone {boneName=pelvis}}
     {targetBone {boneName=thigh_twist_01_l}}
     {targetBone {boneName=thigh_l}}
    ]
    constraintWeights [1,1,1,1]
   }
   n7 = Anim Dynamics - Bone: dyn_zipper_1(ComponentPose=@n6, Alpha=1, ExternalForce="0,0,0") {
    angularDampingOverride=0.9
    angularSpringConstant=2000
    bGravityOverrideInSimSpace=true
    bOverrideAngularDamping=true
    bUseGravityOverride=true
    numSolverIterationsPreUpdate=2
    alphaBoolBlend {blendOption=Linear}
    boundBone {boneName=dyn_zipper_1}
    componentAppliedLinearAccClamp=<3000 3000 3000>
    componentLinearAccScale=<0.01 0.01 0.01>
    componentLinearVelScale=<0.01 0.01 0.01>
    gravityOverride {z=-980}
    physicsBodyDefinitions [{
      sphereCollisionRadius=0
      boundBone {boneName=dyn_zipper_1}
      boxExtents=<4 5 5>
      constraintSetup {angularLimitsMax=<7 8 0>, angularLimitsMin=<-7 -8 -5>, angularTarget=<0 0 0>}
      localJointOffset=<1.5 0 0>
     }]
    relativeSpaceBone {boneName=pelvis}
   }
   Output Pose(Result=Component To Local(ComponentPose=@n7)) {layerGroup=None}
  ]}
 EventGraph {nodes [
   Event BlueprintUpdateAnimation()
   > n1 = Gravity Override State Function DEPRECATED(CharacterPartAnimInstance=Self-Reference(), GameTimeInSeconds=GetGameTimeInSeconds(), GravityOverrideParams=Get Gravity Override Params())
   > Set Gravity Override(Gravity Override=@n1)
  ]}
]
variables [
 body Mesh {type=SkeletalMeshComponent}
 Gravity Override {type=Vector}
 Gravity Override Params {
  type=GravityOverrideParamsStruct
  default {skydive_FloatingDown {gravityOverride=<0 0 1000>, windAmplitude=<0 0 500>, windOffset=<0 0 1000>}}
 }
]
