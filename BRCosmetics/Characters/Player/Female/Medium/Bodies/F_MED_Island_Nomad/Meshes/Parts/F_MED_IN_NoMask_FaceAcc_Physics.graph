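# Physics asset for the no-mask face accessory mesh of the female medium Island Nomad player body, defining dynamics for hair strands and a tentacle bone.
# Capsule bodies on the dyn_ bones form short chains for top, front-left, front-right and back-right hair, each rooted to the head alongside the single tentacle.
# These joints carry per-chain swing limits and a tight twist limit.
# Larger head and neck capsules are joined by a free-rotation constraint.
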PhysicsAsset /BRCosmetics/Characters/Player/Female/Medium/Bodies/F_MED_Island_Nomad/Meshes/Parts/F_MED_IN_NoMask_FaceAcc_Physics
bodies {
 dyn_hairMeshBack_r_01 [{
   shapeType=Capsule
   length=1.75
   radius=0.75
   shapeName=dyn_hairMeshBack_r_01_capsule
   center=<1.5771 0 -0.0001>
   rotation=<rot 90 0.844 0.844>
  }]
 dyn_hairMeshBack_r_02 [{
   shapeType=Capsule
   length=1.75
   radius=0.75
   shapeName=dyn_hairMeshBack_r_02_capsule
   center=<1.5771 0 -0.0001>
   rotation=<rot 90 -1.8984 -1.8983>
  }]
 dyn_hairMeshBack_r_03 [{
   shapeType=Capsule
   length=1.75
   radius=0.75
   shapeName=dyn_hairMeshBack_r_03_capsule
   center=<1.5771 0 -0.0001>
   rotation=<rot 90 -1.2658 -1.2658>
  }]
 dyn_hairMeshFront_l_01 [{
   shapeType=Capsule
   length=4.25
   radius=1
   shapeName=dyn_hairMeshFront_l_01_capsule
   center=<3.45 0 -0.0001>
   rotation=<rot 90 -0.0099 -0.0099>
  }]
 dyn_hairMeshFront_l_02 [{
   shapeType=Capsule
   length=2.5
   radius=0.75
   shapeName=dyn_hairMeshFront_l_02_capsule
   center=<1.556 0.2436 -0.0001>
   rotation=<rot 90 176.4237 171.4772>
  }]
 dyn_hairMeshFront_l_03 [{
   shapeType=Capsule
   length=3
   radius=0.5
   shapeName=dyn_hairMeshFront_l_03_capsule
   center=<2.1679 0 0.3002>
   rotation=<rot 90 -1.2658 -1.2657>
  }]
 dyn_hairMeshFront_l_04 [{
   shapeType=Capsule
   length=1.75
   radius=0.75
   shapeName=dyn_hairMeshFront_l_04_capsule
   center=<1.2692 -0.464 -0.0002>
   rotation=<rot 90 -14.0363 19.5318>
  }]
 dyn_hairMeshFront_l_05 [{
   shapeType=Capsule
   length=1.75
   radius=0.75
   shapeName=dyn_hairMeshFront_l_05_capsule
   center=<1.506 0.2618 -0.0002>
   rotation=<rot 90 0 -15.2074>
  }]
 dyn_hairMeshFront_r_01 [{shapeType=Capsule, length=2.75, radius=0.5, shapeName=dyn_hairMeshFront_r_01_capsule, center=<1.9004 0 -0.1236>, rotation=<rot 84.8194 0 0>}]
 dyn_hairMeshFront_r_02 [{
   shapeType=Capsule
   length=2
   radius=0.4
   shapeName=dyn_hairMeshFront_r_02_capsule
   center=<1.4068 -0.4678 -0.2639>
   rotation=<rot 78.1578 -20.0021 0.0001>
  }]
 dyn_hairTop_01 [{
   shapeType=Capsule
   length=2
   radius=0.5
   shapeName=dyn_hairTop_01_capsule
   center=<1.5144 0.1374 0.3488>
   rotation=<rot 79.814 179.9998 -179.9999>
  }]
 dyn_hairTop_02 [{
   shapeType=Capsule
   length=3
   radius=0.5
   shapeName=dyn_hairTop_02_capsule
   center=<1.5771 -0.0002 -0.3963>
   rotation=<rot 90 -1.2658 -1.2657>
  }]
 dyn_tentacle_1 [{shapeType=Capsule, length=2.5, radius=0.75, shapeName=dyn_tentacle_1_capsule, center=<2.3287 0 -0.0001>, rotation=<rot 90 0.0198 0.0198>}]
 head [
  {shapeType=Capsule, length=7.75, radius=6.25, shapeName=head_capsule, center=<2.9929 1.7429 0>, rotation=<rot 90 90 110.0022>}
  {shapeType=Capsule, length=0.25, radius=7.5, shapeName=head_capsule, center=<6.8046 0 -0.4863>}
 ]
 neck_01 [{shapeType=Capsule, length=4.255, radius=5.505, shapeName=neck_01_capsule, center=<0 1.8253 0>, rotation=<rot 0 89.9999 89.9999>}]
}
constraints [
 {bone1Name=dyn_hairTop_01, bone2Name=head, swing1LimitDegrees=15, swing2LimitDegrees=15, twistLimitDegrees=5}
 {bone1Name=dyn_hairTop_02, bone2Name=dyn_hairTop_01, swing1LimitDegrees=10, swing2LimitDegrees=10, twistLimitDegrees=5}
 {bone1Name=dyn_hairMeshFront_l_01, bone2Name=head, swing1LimitDegrees=25, swing2LimitDegrees=25, twistLimitDegrees=5}
 {bone1Name=dyn_hairMeshFront_l_02, bone2Name=dyn_hairMeshFront_l_01, swing1LimitDegrees=35, swing2LimitDegrees=35, twistLimitDegrees=5}
 {bone1Name=dyn_hairMeshFront_l_05, bone2Name=dyn_hairMeshFront_l_02, swing1LimitDegrees=35, swing2LimitDegrees=35, twistLimitDegrees=5}
 {bone1Name=dyn_hairMeshFront_l_03, bone2Name=head, swing1LimitDegrees=20, swing2LimitDegrees=20, twistLimitDegrees=5}
 {bone1Name=dyn_hairMeshFront_l_04, bone2Name=dyn_hairMeshFront_l_03, swing1LimitDegrees=20, swing2LimitDegrees=20, twistLimitDegrees=5}
 {bone1Name=dyn_hairMeshBack_r_01, bone2Name=head, swing1LimitDegrees=20, swing2LimitDegrees=20, twistLimitDegrees=5}
 {bone1Name=dyn_hairMeshBack_r_02, bone2Name=dyn_hairMeshBack_r_01, swing1LimitDegrees=20, swing2LimitDegrees=20, twistLimitDegrees=5}
 {bone1Name=dyn_hairMeshBack_r_03, bone2Name=dyn_hairMeshBack_r_02, swing1LimitDegrees=20, swing2LimitDegrees=20, twistLimitDegrees=5}
 {bone1Name=dyn_tentacle_1, bone2Name=head, swing1LimitDegrees=20, swing2LimitDegrees=20, twistLimitDegrees=5}
 {bone1Name=dyn_hairMeshFront_r_01, bone2Name=head, swing1LimitDegrees=45, swing2LimitDegrees=45, twistLimitDegrees=5}
 {bone1Name=dyn_hairMeshFront_r_02, bone2Name=dyn_hairMeshFront_r_01, swing1LimitDegrees=45, swing2LimitDegrees=45, twistLimitDegrees=5}
 {
  bone1Name=head
  bone2Name=neck_01
  swing1LimitDegrees=45
  swing1Motion=Free
  swing2LimitDegrees=45
  swing2Motion=Free
  twistLimitDegrees=45
  twistMotion=Free
 }
]
