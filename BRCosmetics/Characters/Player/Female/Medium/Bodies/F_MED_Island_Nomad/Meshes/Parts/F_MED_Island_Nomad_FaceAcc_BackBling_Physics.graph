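# Physics asset for the Island Nomad female medium character's FaceAcc BackBling mesh part, covering a dynamic hair chain and the upper body.
# Four capsules on dyn_hair_03 to 06 form a chain from the head, held by tight swing and twist limits.
# Capsules on pelvis, spine, neck, head and upper arms approximate the body, joined by free-rotation constraints.
# Spine_05 carries three extra offset capsules, enlarging the upper-torso collision volume.
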PhysicsAsset /BRCosmetics/Characters/Player/Female/Medium/Bodies/F_MED_Island_Nomad/Meshes/Parts/F_MED_Island_Nomad_FaceAcc_BackBling_Physics
bodies {
 dyn_hair_03 [{shapeType=Capsule, length=4.25, radius=3, shapeName=dyn_hair_03_capsule, center=<4.097 0 1.0805>, rotation=<rot -90 0 0>}]
 dyn_hair_04 [{shapeType=Capsule, length=4, radius=2.5, shapeName=dyn_hair_04_capsule, center=<4.097 0 0.67>, rotation=<rot -90 0 0>}]
 dyn_hair_05 [{shapeType=Capsule, length=4, radius=2.5, shapeName=dyn_hair_05_capsule, center=<4.097 0 1.4616>, rotation=<rot -90 0 0>}]
 dyn_hair_06 [{shapeType=Capsule, length=4, radius=2.5, shapeName=dyn_hair_06_capsule, center=<4.097 0 -0.2005>, rotation=<rot -90 0 0>}]
 head [
  {shapeType=Capsule, length=6.5, radius=6.25, shapeName=head_capsule, center=<2.7213 0.998 0>, rotation=<rot 90 0 20.0017>}
  {shapeType=Capsule, length=1, radius=7.25, shapeName=head_capsule, center=<7.1126 -0.0226 -0.7311>, rotation=<rot 0 0 91.7749>}
 ]
 neck_01 [{shapeType=Capsule, length=4.255, radius=5.505, shapeName=neck_01_capsule, center=<0 1.8253 0>, rotation=<rot 0 89.9999 89.9999>}]
 pelvis [{shapeType=Capsule, length=7, radius=11.755, shapeName=pelvis_capsule, rotation=<rot 0 -90 0>}]
 spine_01 [{shapeType=Capsule, length=7.005, radius=10.505, shapeName=spine_01_capsule, center=<0 3.4544 0>, rotation=<rot 0 -89.9999 0>}]
 spine_02 [{shapeType=Capsule, length=4.755, radius=10.755, shapeName=spine_02_capsule, center=<0 3.6322 0>, rotation=<rot 0 -89.9999 0>}]
 spine_03 [{shapeType=Capsule, length=4.255, radius=10.005, shapeName=spine_03_capsule, center=<0 2.2914 0>, rotation=<rot 0 -89.9999 0>}]
 spine_04 [{shapeType=Capsule, length=4.505, radius=10.005, shapeName=spine_04_capsule, center=<0 0.6437 0>, rotation=<rot 0 -89.9999 0>}]
 spine_05 [
  {shapeType=Capsule, length=19.005, radius=10.005, shapeName=spine_05_capsule, center=<0 0.7542 0>, rotation=<rot 0 -89.9999 0>}
  {shapeType=Capsule, length=37, radius=8.5, shapeName=spine_05_capsule, center=<3.9421 -10.5394 18.5872>, rotation=<rot 71.1562 39.4099 0>}
  {
   shapeType=Capsule
   length=35.75
   radius=8
   shapeName=spine_05_capsule
   center=<5.0082 -7.5562 -17.2789>
   rotation=<rot 65.6766 -140.5896 179.9999>
  }
  {shapeType=Capsule, length=15, radius=15, shapeName=spine_05_capsule, center=<15.24 -29.1414 -0.0001>}
 ]
 upperarm_l [{
   shapeType=Capsule
   length=18.5
   radius=6.255
   shapeName=upperarm_l_capsule
   center=<11.485 -0.4243 -1.2661>
   rotation=<rot -0.0615 -93.3334 -91.0562>
  }]
 upperarm_r [{
   shapeType=Capsule
   length=19.5
   radius=6.505
   shapeName=upperarm_r_capsule
   center=<-10.8252 0.0968 1.5179>
   rotation=<rot -0.0772 -94.1834 -91.0552>
  }]
}
constraints [
 {
  bone1Name=head
  bone2Name=neck_01
  swing1LimitDegrees=45
  swing1Motion=Free
  swing2LimitDegrees=45
  swing2Motion=Free
  twistLimitDegrees=45
  twistMotion=Free
 }
 {
  bone1Name=spine_05
  bone2Name=spine_04
  swing1LimitDegrees=45
  swing1Motion=Free
  swing2LimitDegrees=45
  swing2Motion=Free
  twistLimitDegrees=45
  twistMotion=Free
 }
 {
  bone1Name=spine_04
  bone2Name=spine_03
  swing1LimitDegrees=45
  swing1Motion=Free
  swing2LimitDegrees=45
  swing2Motion=Free
  twistLimitDegrees=45
  twistMotion=Free
 }
 {
  bone1Name=neck_01
  bone2Name=spine_05
  swing1LimitDegrees=45
  swing1Motion=Free
  swing2LimitDegrees=45
  swing2Motion=Free
  twistLimitDegrees=45
  twistMotion=Free
 }
 {
  bone1Name=spine_02
  bone2Name=spine_01
  swing1LimitDegrees=45
  swing1Motion=Free
  swing2LimitDegrees=45
  swing2Motion=Free
  twistLimitDegrees=45
  twistMotion=Free
 }
 {
  bone1Name=spine_03
  bone2Name=spine_02
  swing1LimitDegrees=45
  swing1Motion=Free
  swing2LimitDegrees=45
  swing2Motion=Free
  twistLimitDegrees=45
  twistMotion=Free
 }
 {
  bone1Name=upperarm_l
  bone2Name=spine_05
  swing1LimitDegrees=45
  swing1Motion=Free
  swing2LimitDegrees=45
  swing2Motion=Free
  twistLimitDegrees=45
  twistMotion=Free
 }
 {
  bone1Name=upperarm_r
  bone2Name=spine_05
  swing1LimitDegrees=45
  swing1Motion=Free
  swing2LimitDegrees=45
  swing2Motion=Free
  twistLimitDegrees=45
  twistMotion=Free
 }
 {bone1Name=dyn_hair_03, bone2Name=head, swing1LimitDegrees=10, swing2LimitDegrees=10, twistLimitDegrees=20}
 {bone1Name=dyn_hair_04, bone2Name=dyn_hair_03, swing1LimitDegrees=10, swing2LimitDegrees=20, twistLimitDegrees=20}
 {bone1Name=dyn_hair_05, bone2Name=dyn_hair_04, swing1LimitDegrees=10, swing2LimitDegrees=20, twistLimitDegrees=2}
 {bone1Name=dyn_hair_06, bone2Name=dyn_hair_05, swing1LimitDegrees=10, swing2LimitDegrees=20, twistLimitDegrees=2}
 {
  bone1Name=spine_01
  bone2Name=pelvis
  swing1LimitDegrees=45
  swing1Motion=Free
  swing2LimitDegrees=45
  swing2Motion=Free
  twistLimitDegrees=45
  twistMotion=Free
 }
]
