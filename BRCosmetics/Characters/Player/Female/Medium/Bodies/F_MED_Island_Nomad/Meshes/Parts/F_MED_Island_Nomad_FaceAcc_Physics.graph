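# Physics asset for the face accessory part of the female medium Island Nomad player cosmetic, setting up a dynamic hair chain.
# Six dyn_hair capsules hang in sequence from the head, generally narrowing toward the tip, with limited swing and twist that loosen down the chain.
# Head, neck, spine and upper-arm capsules form the body volume around the hair, with all joints locked except a fully free head-to-neck rotation.
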PhysicsAsset /BRCosmetics/Characters/Player/Female/Medium/Bodies/F_MED_Island_Nomad/Meshes/Parts/F_MED_Island_Nomad_FaceAcc_Physics
bodies {
 dyn_hair_01 [{shapeType=Capsule, length=1.25, radius=3.25, shapeName=dyn_hair_01_capsule, center=<9.8025 -0.0001 1.1083>, rotation=<rot -89.3854 180 -180>}]
 dyn_hair_02 [{shapeType=Capsule, length=3, radius=3.5, shapeName=dyn_hair_02_capsule, center=<4.097 -0.0001 1.5896>, rotation=<rot -90 0 0>}]
 dyn_hair_03 [{shapeType=Capsule, length=4.5, radius=2.75, shapeName=dyn_hair_03_capsule, center=<4.097 0 0.7441>, rotation=<rot -90 180 180>}]
 dyn_hair_04 [{shapeType=Capsule, length=4.25, radius=2.25, shapeName=dyn_hair_04_capsule, center=<4.097 0.0001 0.2633>, rotation=<rot -90 0 0>}]
 dyn_hair_05 [{shapeType=Capsule, length=5.25, radius=1.75, shapeName=dyn_hair_05_capsule, center=<4.097 0 -0.0854>, rotation=<rot -90 0 0>}]
 dyn_hair_06 [{shapeType=Capsule, length=5.75, radius=1.75, shapeName=dyn_hair_06_capsule, center=<5.1415 0 -1.0721>, rotation=<rot -90 0 0>}]
 head [
  {shapeType=Capsule, length=9.75, radius=6.75, shapeName=head_capsule, center=<2.9511 1.5624 -0.0001>, rotation=<rot 90 172.875 -164.9797>}
  {shapeType=Capsule, length=1, radius=7.25, shapeName=head_capsule, center=<7.1126 -0.0226 -0.7311>, rotation=<rot 0 0 91.7749>}
 ]
 neck_01 [{shapeType=Capsule, length=5.505, radius=5.755, shapeName=neck_01_capsule, center=<3.4607 -0.0909 -0.0001>, rotation=<rot 0 88.5361 89.9998>}]
 spine_01 [{shapeType=Capsule, length=6.755, radius=12.255, shapeName=spine_01_capsule, center=<-6.0737 0.2981 0>, rotation=<rot 0 -89.9999 0>}]
 spine_02 [{shapeType=Capsule, length=4.755, radius=11.505, shapeName=spine_02_capsule, center=<0 3.6322 0>, rotation=<rot 0 -89.9999 0>}]
 spine_03 [{shapeType=Capsule, length=3.255, radius=11.505, shapeName=spine_03_capsule, center=<0 2.2914 0>, rotation=<rot 0 -89.9999 0>}]
 spine_04 [{shapeType=Capsule, length=2.505, radius=12.005, shapeName=spine_04_capsule, center=<0 1.3782 0>, rotation=<rot 0 -89.9999 0>}]
 spine_05 [{shapeType=Capsule, length=15.005, radius=10.755, shapeName=spine_05_capsule, center=<0.1553 0.7709 0>, rotation=<rot 0 -89.9999 0>}]
 upperarm_l [{
   shapeType=Capsule
   length=18.5
   radius=6.255
   shapeName=upperarm_l_capsule
   center=<11.485 -0.4243 -1.2661>
   rotation=<rot -0.0615 -93.3334 -91.0562>
  }]
 upperarm_r [{
   shapeType=Capsule
   length=19.5
   radius=6.505
   shapeName=upperarm_r_capsule
   center=<-10.8252 0.0968 1.5179>
   rotation=<rot -0.0772 -94.1834 -91.0552>
  }]
}
constraints [
 {bone1Name=dyn_hair_01, bone2Name=head, swing1LimitDegrees=12, swing2LimitDegrees=50, twistLimitDegrees=2}
 {bone1Name=dyn_hair_02, bone2Name=dyn_hair_01, swing1LimitDegrees=12, swing2LimitDegrees=45, twistLimitDegrees=5}
 {bone1Name=dyn_hair_03, bone2Name=dyn_hair_02, swing1LimitDegrees=12, swing2LimitDegrees=45, twistLimitDegrees=20}
 {bone1Name=dyn_hair_04, bone2Name=dyn_hair_03, swing1LimitDegrees=45, swing2LimitDegrees=45, twistLimitDegrees=20}
 {bone1Name=dyn_hair_05, bone2Name=dyn_hair_04, swing1LimitDegrees=45, swing2LimitDegrees=45, twistLimitDegrees=20}
 {bone1Name=dyn_hair_06, bone2Name=dyn_hair_05, swing1LimitDegrees=45, swing2LimitDegrees=45, twistLimitDegrees=20}
 {
  bone1Name=head
  bone2Name=neck_01
  swing1LimitDegrees=45
  swing1Motion=Free
  swing2LimitDegrees=45
  swing2Motion=Free
  twistLimitDegrees=45
  twistMotion=Free
 }
 {
  bone1Name=spine_05
  bone2Name=spine_04
  swing1LimitDegrees=45
  swing1Motion=Locked
  swing2LimitDegrees=45
  swing2Motion=Locked
  twistLimitDegrees=45
  twistMotion=Locked
 }
 {
  bone1Name=spine_04
  bone2Name=spine_03
  swing1LimitDegrees=45
  swing1Motion=Locked
  swing2LimitDegrees=45
  swing2Motion=Locked
  twistLimitDegrees=45
  twistMotion=Locked
 }
 {
  bone1Name=neck_01
  bone2Name=spine_05
  swing1LimitDegrees=45
  swing1Motion=Locked
  swing2LimitDegrees=45
  swing2Motion=Locked
  twistLimitDegrees=45
  twistMotion=Locked
 }
 {
  bone1Name=spine_02
  bone2Name=spine_01
  swing1LimitDegrees=45
  swing1Motion=Locked
  swing2LimitDegrees=45
  swing2Motion=Locked
  twistLimitDegrees=45
  twistMotion=Locked
 }
 {
  bone1Name=spine_03
  bone2Name=spine_02
  swing1LimitDegrees=45
  swing1Motion=Locked
  swing2LimitDegrees=45
  swing2Motion=Locked
  twistLimitDegrees=45
  twistMotion=Locked
 }
 {
  bone1Name=upperarm_l
  bone2Name=spine_05
  swing1LimitDegrees=45
  swing1Motion=Locked
  swing2LimitDegrees=45
  swing2Motion=Locked
  twistLimitDegrees=45
  twistMotion=Locked
 }
 {
  bone1Name=upperarm_r
  bone2Name=spine_05
  swing1LimitDegrees=45
  swing1Motion=Locked
  swing2LimitDegrees=45
  swing2Motion=Locked
  twistLimitDegrees=45
  twistMotion=Locked
 }
]
