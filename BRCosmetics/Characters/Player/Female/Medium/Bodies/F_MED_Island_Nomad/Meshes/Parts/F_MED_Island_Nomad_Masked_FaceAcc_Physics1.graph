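# Physics asset for the female Island Nomad masked face-accessory mesh, covering secondary motion of hair, earrings, feathers, fluff, pendulums and a tentacle.
# Capsule bodies on dyn_ bones, spanning several named accessory sets (cat, bird, demon, fox, owl, Butter Cake), hang from the head singly or in short chains with limited swing and little or no twist.
# Head, neck, upper-spine and upper-arm capsules form the body volume, linked by free angular constraints.
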PhysicsAsset /BRCosmetics/Characters/Player/Female/Medium/Bodies/F_MED_Island_Nomad/Meshes/Parts/F_MED_Island_Nomad_Masked_FaceAcc_Physics1
bodies {
 dyn_BUTTER_CAKE_RED_earring_l [{shapeType=Capsule, length=0.5, radius=1.25, shapeName=dyn_BUTTER_CAKE_RED_earring_l_capsule, center=<1.7288 0 0>, rotation=<rot 90 0 0>}]
 dyn_BUTTER_CAKE_RED_earring_r [{shapeType=Capsule, length=0.5, radius=1.25, shapeName=dyn_BUTTER_CAKE_RED_earring_r_capsule, center=<1.7288 0 0>, rotation=<rot 90 0 0>}]
 dyn_BUTTER_CAKE_earring_l [{shapeType=Capsule, length=1, radius=1, shapeName=dyn_BUTTER_CAKE_earring_l_capsule, center=<2.8217 0 0>, rotation=<rot 90 0 0>}]
 dyn_BUTTER_CAKE_earring_r [{shapeType=Capsule, length=1, radius=1, shapeName=dyn_BUTTER_CAKE_earring_r_capsule, center=<2.8481 0 0>, rotation=<rot 90 0 0>}]
 dyn_FOX_Pendulum_A_1 [{shapeType=Capsule, length=0.755, radius=0.755, shapeName=dyn_birdFeather_1_capsule, rotation=<rot -90 0 0>}]
 dyn_FOX_Pendulum_A_2 [{shapeType=Capsule, length=4.005, radius=0.755, shapeName=dyn_birdFeather_1_capsule, center=<2.6638 0 0>, rotation=<rot -90 0 0>}]
 dyn_FOX_Pendulum_B_1 [{shapeType=Capsule, length=0.755, radius=0.755, shapeName=dyn_birdFeather_1_capsule, rotation=<rot -90 0 0>}]
 dyn_FOX_Pendulum_B_2 [{shapeType=Capsule, length=4.005, radius=0.755, shapeName=dyn_birdFeather_1_capsule, center=<2.6638 0 0>, rotation=<rot -90 0 0>}]
 dyn_OWL_feather_A_1 [{shapeType=Capsule, length=2.755, radius=1.005, shapeName=dyn_BUTTER_CAKE_RED_earring_r_capsule, center=<3.5 0 0>, rotation=<rot 90 0 0>}]
 dyn_OWL_feather_A_2 [{
   shapeType=Capsule
   length=10.005
   radius=1.005
   shapeName=dyn_OWL_feather_A_2_capsule
   center=<4.2 -0.2 1>
   rotation=<rot 74.9273 179.9999 -176.0647>
  }]
 dyn_OWL_feather_B_1 [{shapeType=Capsule, length=0.755, radius=0.755, shapeName=dyn_BUTTER_CAKE_RED_earring_r_capsule, center=<0 0 0.5>}]
 dyn_OWL_feather_B_2 [{shapeType=Capsule, length=0.755, radius=0.755, shapeName=dyn_BUTTER_CAKE_RED_earring_r_capsule, center=<0.5 0 0>}]
 dyn_OWL_feather_B_3 [{shapeType=Capsule, length=0.755, radius=0.755, shapeName=dyn_BUTTER_CAKE_RED_earring_r_capsule, center=<-0.2 0 0.4>}]
 dyn_birdFeather_1 [{shapeType=Capsule, length=0.255, radius=1.255, shapeName=dyn_birdFeather_1_capsule, center=<1.1679 0 0>, rotation=<rot -90 0 0>}]
 dyn_birdFeather_2 [{shapeType=Capsule, length=2.005, radius=1.255, shapeName=dyn_birdFeather_2_capsule, center=<4.2904 0 0>, rotation=<rot -90 0 0>}]
 dyn_catEarring_1 [{shapeType=Capsule, length=0, radius=1.25, shapeName=dyn_catEarring_1_capsule, center=<-0.4 0 0>, rotation=<rot 0 89.9999 0>}]
 dyn_catEarring_2 [{shapeType=Capsule, length=0, radius=1.25, shapeName=dyn_catEarring_2_capsule, center=<-0.4 0 0>, rotation=<rot 0 89.9999 0>}]
 dyn_catFeather_1 [{shapeType=Capsule, length=0, radius=2.25, shapeName=dyn_catFeather_1_capsule, rotation=<rot 0 -90 0>}]
 dyn_catFeather_2 [{shapeType=Capsule, length=1.505, radius=1.005, shapeName=dyn_catFeather_2_capsule, center=<0 0 -2.7382>, rotation=<rot 0 -89.9999 0>}]
 dyn_catFeather_3 [{shapeType=Capsule, length=4.505, radius=1.755, shapeName=dyn_catFeather_3_capsule, center=<3.3834 0 0>, rotation=<rot 0 -89.9999 -89.9999>}]
 dyn_catFeather_4 [{
   shapeType=Capsule
   length=2.505
   radius=1.005
   shapeName=dyn_catFeather_4_capsule
   center=<1.974 0 -0.9255>
   rotation=<rot 0 -89.9998 -63.0743>
  }]
 dyn_demonFluff_c_1 [{shapeType=Capsule, length=0, radius=2.005, shapeName=dyn_demonFluff_c_1_capsule, center=<0.3804 -0.0554 0>, rotation=<rot 0 -89.9999 0>}]
 dyn_demonFluff_l_1 [{shapeType=Capsule, length=0, radius=1.755, shapeName=dyn_demonFluff_l_1_capsule, center=<0.8602 -0.0002 0>, rotation=<rot 0 -89.9999 0>}]
 dyn_demonFluff_l_2 [{shapeType=Capsule, length=0, radius=1.755, shapeName=dyn_demonFluff_l_2_capsule, center=<0.8602 -0.0002 0>, rotation=<rot 0 -89.9999 0>}]
 dyn_demonFluff_l_3 [{shapeType=Capsule, length=0, radius=1.755, shapeName=dyn_demonFluff_l_3_capsule, center=<0.8602 -0.0002 0>, rotation=<rot 0 -89.9999 0>}]
 dyn_demonFluff_r_1 [{shapeType=Capsule, length=0, radius=1.755, shapeName=dyn_demonFluff_r_1_capsule, center=<0.8602 -0.0002 0>, rotation=<rot 0 -89.9999 0>}]
 dyn_demonFluff_r_2 [{shapeType=Capsule, length=0, radius=1.755, shapeName=dyn_demonFluff_r_2_capsule, center=<0.8602 -0.0002 0>, rotation=<rot 0 -89.9999 0>}]
 dyn_demonFluff_r_3 [{shapeType=Capsule, length=0, radius=1.755, shapeName=dyn_demonFluff_r_3_capsule, center=<0.8602 -0.0002 0>, rotation=<rot 0 -89.9999 0>}]
 dyn_hairMeshBack_r_01 [{
   shapeType=Capsule
   length=1.75
   radius=0.75
   shapeName=dyn_hairMeshBack_r_01_capsule
   center=<1.5771 0 -0.0001>
   rotation=<rot 90 0.844 0.844>
  }]
 dyn_hairMeshBack_r_02 [{
   shapeType=Capsule
   length=1.75
   radius=0.75
   shapeName=dyn_hairMeshBack_r_02_capsule
   center=<1.5771 0 -0.0001>
   rotation=<rot 90 -1.8984 -1.8983>
  }]
 dyn_hairMeshBack_r_03 [{
   shapeType=Capsule
   length=1.75
   radius=0.75
   shapeName=dyn_hairMeshBack_r_03_capsule
   center=<1.5771 0 -0.0001>
   rotation=<rot 90 -1.2658 -1.2658>
  }]
 dyn_hairMeshFront_l_01 [{
   shapeType=Capsule
   length=4.5
   radius=0.75
   shapeName=dyn_hairMeshFront_l_01_capsule
   center=<3.45 0 -0.0001>
   rotation=<rot 90 -0.0099 -0.0099>
  }]
 dyn_hairMeshFront_l_02 [{
   shapeType=Capsule
   length=2.75
   radius=0.6
   shapeName=dyn_hairMeshFront_l_02_capsule
   center=<1.568 0.1057 -0.281>
   rotation=<rot 90 3.5763 -1.3701>
  }]
 dyn_hairMeshFront_l_03 [{
   shapeType=Capsule
   length=2.75
   radius=0.75
   shapeName=dyn_hairMeshFront_l_03_capsule
   center=<2.1679 0 -0.0001>
   rotation=<rot 90 -1.2658 -1.2658>
  }]
 dyn_hairMeshFront_l_04 [{
   shapeType=Capsule
   length=1.75
   radius=0.75
   shapeName=dyn_hairMeshFront_l_04_capsule
   center=<1.2692 -0.464 -0.0002>
   rotation=<rot 90 -14.0363 19.5318>
  }]
 dyn_hairMeshFront_l_05 [{
   shapeType=Capsule
   length=1.75
   radius=0.75
   shapeName=dyn_hairMeshFront_l_05_capsule
   center=<1.506 0.2618 -0.0002>
   rotation=<rot 90 0 -15.2074>
  }]
 dyn_hairMeshFront_r_01 [{shapeType=Capsule, length=2.75, radius=0.5, shapeName=dyn_hairMeshFront_r_01_capsule, center=<1.8908 0 -0.2295>, rotation=<rot 84.8194 0 0>}]
 dyn_hairMeshFront_r_02 [{
   shapeType=Capsule
   length=2
   radius=0.5
   shapeName=dyn_hairMeshFront_r_02_capsule
   center=<1.3481 -0.4464 -0.5608>
   rotation=<rot 78.1579 -20.0021 0.0001>
  }]
 dyn_hairTop_01 [{
   shapeType=Capsule
   length=1.75
   radius=0.75
   shapeName=dyn_hairTop_01_capsule
   center=<1.5144 -0.0002 0.3488>
   rotation=<rot 79.814 179.9998 -179.9999>
  }]
 dyn_tentacle_1 [{shapeType=Capsule, length=2.5, radius=0.75, shapeName=dyn_tentacle_1_capsule, center=<2.3287 0 -0.0001>, rotation=<rot 90 0.0198 0.0198>}]
 head [
  {shapeType=Capsule, length=6.5, radius=6.25, shapeName=head_capsule, center=<3.7162 3.7298 0>, rotation=<rot 90 0 20.003>}
  {shapeType=Capsule, length=1, radius=7.25, shapeName=head_capsule, center=<7.1126 -0.0226 -0.7311>, rotation=<rot 0 0 91.7749>}
  {
   shapeType=Capsule
   length=8.75
   radius=4.25
   shapeName=head_capsule
   center=<-3.7111 8.1999 -10.5252>
   rotation=<rot 28.8483 15.4162 42.6472>
  }
  {shapeType=Capsule, length=11.75, radius=4.25, shapeName=head_capsule, center=<3.3125 11.4941 -7.8375>, rotation=<rot 35.5144 0 25.4563>}
  {shapeType=Capsule, length=7.5, radius=5.25, shapeName=head_capsule, center=<3.2453 8.644 0.698>, rotation=<rot 90 -0.0099 -0.0099>}
  {shapeType=Capsule, length=2.25, radius=1.25, shapeName=head_capsule, center=<7.7619 1.8966 -12.1106>}
  {shapeType=Capsule, length=1.75, radius=1.25, shapeName=head_capsule, center=<7.9927 1.8763 12.0408>}
 ]
 neck_01 [{shapeType=Capsule, length=4.255, radius=5.505, shapeName=neck_01_capsule, center=<0 1.8253 0>, rotation=<rot 0 89.9999 89.9999>}]
 spine_05 [
  {shapeType=Capsule, length=7.005, radius=10.005, shapeName=spine_05_capsule, center=<-1.5316 -0.1881 0>, rotation=<rot 0 -100.2616 0>}
  {shapeType=Capsule, length=11.5, radius=8.75, shapeName=spine_05_capsule, center=<1.793 -2.4478 4.1754>}
 ]
 upperarm_l [{
   shapeType=Capsule
   length=18.5
   radius=6.255
   shapeName=upperarm_l_capsule
   center=<11.485 -0.4243 -1.2661>
   rotation=<rot -0.0615 -93.3334 -91.0562>
  }]
 upperarm_r [{
   shapeType=Capsule
   length=19.5
   radius=6.505
   shapeName=upperarm_r_capsule
   center=<-10.8252 0.0968 1.5179>
   rotation=<rot -0.0772 -94.1834 -91.0552>
  }]
}
constraints [
 {bone1Name=dyn_hairTop_01, bone2Name=head, swing1LimitDegrees=15, swing2LimitDegrees=15, twistLimitDegrees=5}
 {bone1Name=dyn_hairMeshFront_l_01, bone2Name=head, swing1LimitDegrees=10, swing2LimitDegrees=10, twistLimitDegrees=2}
 {bone1Name=dyn_hairMeshFront_l_02, bone2Name=dyn_hairMeshFront_l_01, swing1LimitDegrees=10, swing2LimitDegrees=10, twistLimitDegrees=2}
 {bone1Name=dyn_hairMeshFront_l_05, bone2Name=dyn_hairMeshFront_l_02, swing1LimitDegrees=10, swing2LimitDegrees=10, twistLimitDegrees=2}
 {bone1Name=dyn_hairMeshFront_l_03, bone2Name=head, swing1LimitDegrees=20, swing2LimitDegrees=20, twistLimitDegrees=5}
 {bone1Name=dyn_hairMeshFront_l_04, bone2Name=dyn_hairMeshFront_l_03, swing1LimitDegrees=20, swing2LimitDegrees=20, twistLimitDegrees=5}
 {bone1Name=dyn_hairMeshBack_r_01, bone2Name=head, swing1LimitDegrees=20, swing2LimitDegrees=20, twistLimitDegrees=5}
 {bone1Name=dyn_hairMeshBack_r_02, bone2Name=dyn_hairMeshBack_r_01, swing1LimitDegrees=20, swing2LimitDegrees=20, twistLimitDegrees=5}
 {bone1Name=dyn_hairMeshBack_r_03, bone2Name=dyn_hairMeshBack_r_02, swing1LimitDegrees=20, swing2LimitDegrees=20, twistLimitDegrees=5}
 {bone1Name=dyn_tentacle_1, bone2Name=head, swing1LimitDegrees=20, swing2LimitDegrees=20, twistLimitDegrees=5}
 {bone1Name=dyn_hairMeshFront_r_01, bone2Name=head, swing1LimitDegrees=10, swing2LimitDegrees=10, twistLimitDegrees=0}
 {bone1Name=dyn_hairMeshFront_r_02, bone2Name=dyn_hairMeshFront_r_01, swing1LimitDegrees=10, swing2LimitDegrees=10, twistLimitDegrees=0}
 {
  bone1Name=head
  bone2Name=neck_01
  swing1LimitDegrees=45
  swing1Motion=Free
  swing2LimitDegrees=45
  swing2Motion=Free
  twistLimitDegrees=45
  twistMotion=Free
 }
 {
  bone1Name=neck_01
  bone2Name=spine_05
  swing1LimitDegrees=45
  swing1Motion=Free
  swing2LimitDegrees=45
  swing2Motion=Free
  twistLimitDegrees=45
  twistMotion=Free
 }
 {
  bone1Name=upperarm_l
  bone2Name=spine_05
  swing1LimitDegrees=45
  swing1Motion=Free
  swing2LimitDegrees=45
  swing2Motion=Free
  twistLimitDegrees=45
  twistMotion=Free
 }
 {
  bone1Name=upperarm_r
  bone2Name=spine_05
  swing1LimitDegrees=45
  swing1Motion=Free
  swing2LimitDegrees=45
  swing2Motion=Free
  twistLimitDegrees=45
  twistMotion=Free
 }
 {bone1Name=dyn_catFeather_1, bone2Name=head, swing1LimitDegrees=20, swing2LimitDegrees=20, twistLimitDegrees=0}
 {bone1Name=dyn_catFeather_2, bone2Name=dyn_catFeather_1, swing1LimitDegrees=25, swing2LimitDegrees=25, twistLimitDegrees=0}
 {bone1Name=dyn_catFeather_3, bone2Name=dyn_catFeather_2, swing1LimitDegrees=25, swing2LimitDegrees=25, twistLimitDegrees=0}
 {bone1Name=dyn_catFeather_4, bone2Name=dyn_catFeather_3, swing1LimitDegrees=25, swing2LimitDegrees=25, twistLimitDegrees=0}
 {bone1Name=dyn_birdFeather_1, bone2Name=head, swing1LimitDegrees=20, swing2LimitDegrees=20, twistLimitDegrees=0}
 {bone1Name=dyn_birdFeather_2, bone2Name=dyn_birdFeather_1, swing1LimitDegrees=20, swing2LimitDegrees=20, twistLimitDegrees=0}
 {bone1Name=dyn_demonFluff_l_3, bone2Name=head, swing1LimitDegrees=22, swing2LimitDegrees=3, twistLimitDegrees=0}
 {bone1Name=dyn_demonFluff_l_1, bone2Name=head, swing1LimitDegrees=22, swing2LimitDegrees=15, twistLimitDegrees=0}
 {bone1Name=dyn_demonFluff_l_2, bone2Name=head, swing1LimitDegrees=22, swing2LimitDegrees=15, twistLimitDegrees=0}
 {bone1Name=dyn_demonFluff_c_1, bone2Name=head, swing1LimitDegrees=22, swing2LimitDegrees=15, twistLimitDegrees=0}
 {bone1Name=dyn_demonFluff_r_3, bone2Name=head, swing1LimitDegrees=22, swing2LimitDegrees=3, twistLimitDegrees=0}
 {bone1Name=dyn_demonFluff_r_1, bone2Name=head, swing1LimitDegrees=22, swing2LimitDegrees=15, twistLimitDegrees=0}
 {bone1Name=dyn_demonFluff_r_2, bone2Name=head, swing1LimitDegrees=22, swing2LimitDegrees=15, twistLimitDegrees=0}
 {bone1Name=dyn_FOX_Pendulum_A_2, bone2Name=dyn_FOX_Pendulum_A_1, swing1LimitDegrees=30, swing2LimitDegrees=30, twistLimitDegrees=0}
 {bone1Name=dyn_FOX_Pendulum_A_1, bone2Name=head, swing1LimitDegrees=30, swing2LimitDegrees=30, twistLimitDegrees=0}
 {bone1Name=dyn_FOX_Pendulum_B_1, bone2Name=head, swing1LimitDegrees=30, swing2LimitDegrees=30, twistLimitDegrees=0}
 {bone1Name=dyn_FOX_Pendulum_B_2, bone2Name=dyn_FOX_Pendulum_B_1, swing1LimitDegrees=30, swing2LimitDegrees=30, twistLimitDegrees=0}
 {bone1Name=dyn_BUTTER_CAKE_earring_l, bone2Name=head, swing1LimitDegrees=90, swing2LimitDegrees=55, twistLimitDegrees=0}
 {bone1Name=dyn_BUTTER_CAKE_earring_r, bone2Name=head, swing1LimitDegrees=90, swing2LimitDegrees=55, twistLimitDegrees=0}
 {bone1Name=dyn_BUTTER_CAKE_RED_earring_l, bone2Name=head, swing1LimitDegrees=50, swing2LimitDegrees=20, twistLimitDegrees=0}
 {bone1Name=dyn_BUTTER_CAKE_RED_earring_r, bone2Name=head, swing1LimitDegrees=20, swing2LimitDegrees=50, twistLimitDegrees=0}
 {bone1Name=dyn_OWL_feather_B_1, bone2Name=head, swing1LimitDegrees=20, swing2LimitDegrees=20, twistLimitDegrees=0}
 {bone1Name=dyn_OWL_feather_B_2, bone2Name=head, swing1LimitDegrees=20, swing2LimitDegrees=20, twistLimitDegrees=0}
 {bone1Name=dyn_OWL_feather_B_3, bone2Name=head, swing1LimitDegrees=20, swing2LimitDegrees=20, twistLimitDegrees=0}
 {bone1Name=dyn_OWL_feather_A_1, bone2Name=head, swing1LimitDegrees=20, swing2LimitDegrees=20, twistLimitDegrees=0}
 {bone1Name=dyn_OWL_feather_A_2, bone2Name=dyn_OWL_feather_A_1, swing1LimitDegrees=5, swing2LimitDegrees=5, twistLimitDegrees=0}
 {bone1Name=dyn_catEarring_1, bone2Name=head, swing1LimitDegrees=0, swing2LimitDegrees=22, twistLimitDegrees=0}
 {bone1Name=dyn_catEarring_2, bone2Name=head, swing1LimitDegrees=0, swing2LimitDegrees=22, twistLimitDegrees=0}
]
