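# Post-process-style AnimBP for the Island Nomad face accessories, simulating hair and mask motion over the pose copied from the parent mesh.
# A cached RigidBody pose feeds masked and unmasked simulations, then deer-mask leaf dynamics, an aiming adjustment, and backpack hair constraints, head-turn offsets and physics.
# The event graph derives these controls from cosmetic variant, backbling, head twist, gravity and jump state.
# Cat-earring and skydive hair nodes are unconnected to the output.
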AnimBlueprint /BRCosmetics/Characters/Player/Female/Medium/Bodies/F_MED_Island_Nomad/Meshes/Parts/F_MED_Island_Nomad_FaceAcc_AnimBP
target_skeleton=&/F_MED_Island_Nomad_FaceAcc_Skeleton
graphs [
 AnimGraph {nodes [
   n1 = Anim Dynamics - Bone: dyn_catEarring_1(Alpha=0, ExternalForce="0,0,0") {
    angularDampingOverride=0.8
    angularSpringConstant=2000
    bGravityOverrideInSimSpace=true
    bOverrideAngularDamping=true
    bUseGravityOverride=true
    numSolverIterationsPreUpdate=2
    simulationSpace=BoneRelative
    alphaBoolBlend {blendOption=Linear}
    boundBone {boneName=dyn_catEarring_1}
    componentAppliedLinearAccClamp=<8000 8000 8000>
    componentLinearAccScale=<5 5 5>
    componentLinearVelScale=<5 5 5>
    gravityOverride {z=-980}
    physicsBodyDefinitions [{
      sphereCollisionRadius=0
      boundBone {boneName=dyn_catEarring_1}
      boxExtents=<2 5 5>
      constraintSetup {angularLimitsMax=<0 30 0>, angularLimitsMin=<0 -30 0>, angularTarget=<0 -10 0>}
      localJointOffset=<0 0 0>
     }]
    relativeSpaceBone {boneName=neck_01}
   }
   Anim Dynamics - Bone: dyn_catEarring_2(ComponentPose=@n1, Alpha=0, ExternalForce="0,0,0") {
    angularDampingOverride=0.8
    angularSpringConstant=2000
    bGravityOverrideInSimSpace=true
    bOverrideAngularDamping=true
    bUseGravityOverride=true
    numSolverIterationsPreUpdate=2
    simulationSpace=BoneRelative
    alphaBoolBlend {blendOption=Linear}
    boundBone {boneName=dyn_catEarring_2}
    componentAppliedLinearAccClamp=<8000 8000 8000>
    componentLinearAccScale=<5 5 5>
    componentLinearVelScale=<5 5 5>
    gravityOverride {z=-980}
    physicsBodyDefinitions [{
      sphereCollisionRadius=0
      boundBone {boneName=dyn_catEarring_2}
      boxExtents=<2 5 5>
      constraintSetup {angularLimitsMax=<0 30 0>, angularLimitsMin=<0 -30 0>, angularTarget=<0 -10 0>}
      localJointOffset=<0 0 0>
     }]
    relativeSpaceBone {boneName=neck_01}
   }
   n2 = Get skydive float() -> (skydive float:double)
   n3 = "Constraint - dyn_hair_01 (P)"(Alpha=@n2, ConstraintWeights_0=1) {
    alphaBoolBlend {blendOption=Linear}
    boneToModify {boneName=dyn_hair_01}
    constraintSetup [{targetBone {boneName=spine_05}}]
    constraintWeights [1]
   }
   n4 = "Transform (Modify) Bone - Bone: dyn_hair_01"(ComponentPose=@n3, Alpha=@n2, Translation="5,0,0", Rotation="0,0,0", Scale="1,1,1") {
    rotationMode=BMM_Additive
    rotationSpace=BCS_BoneSpace
    translationMode=BMM_Additive
    translationSpace=BCS_BoneSpace
    alphaBoolBlend {blendOption=Linear}
    boneToModify {boneName=dyn_hair_01}
    translation {x=5}
   }
   "Transform (Modify) Bone - Bone: dyn_hair_04"(ComponentPose=@n4, Alpha=@n2, Translation="0,0,0", Rotation="10,0,0", Scale="1,1,1") {
    rotationMode=BMM_Additive
    rotationSpace=BCS_BoneSpace
    translationMode=BMM_Additive
    alphaBoolBlend {blendOption=Linear}
    boneToModify {boneName=dyn_hair_04}
    rotation {pitch=10}
   }
   n5 = Get jumping Acc values Braid() -> (jumping Acc values Braid:Vector)
   n6 = RigidBody!(
    ComponentPose=Local To Component(LocalPose=Use cached pose 'defaultPose'())
    Alpha
    OverrideWorldGravity:Vector=Get Gravity Override()
    ExternalForce
    ComponentLinearAccScale:Vector=@n5
    ComponentLinearVelScale:Vector=@n5
   ) -> (Pose) {
    bClampLinearTranslationLimitToRefPose=true
    bOverrideWorldGravity=true
    overridePhysicsAsset=&/F_MED_IN_NoMask_FaceAcc_Physics
    simulationSpace=BaseBoneSpace
    alphaBoolBlend {blendOption=Linear}
    baseBoneRef {boneName=head}
    componentAppliedLinearAccClamp=<500 500 500>
    componentLinearAccScale=<0.001 0.001 0.001>
    componentLinearVelScale=<0.001 0.001 0.001>
    overrideWorldGravity {z=-980}
   }
   n7 = Component To Local(ComponentPose=@n6)
   n8 = Get jumping Acc Scale Values() -> (jumping Acc Scale Values:Vector)
   n9 = RigidBody(
    ComponentPose=Local To Component(LocalPose=Use cached pose 'defaultPose'())
    OverrideWorldGravity=Get Gravity Override()
    ComponentLinearAccScale=@n8
    ComponentLinearVelScale=@n8
   ) {
    bClampLinearTranslationLimitToRefPose=true
    bOverrideWorldGravity=true
    overridePhysicsAsset=&/F_MED_Island_Nomad_Masked_FaceAcc_Physics1
    simulationSpace=BaseBoneSpace
    alphaBoolBlend {blendOption=Linear}
    baseBoneRef {boneName=head}
    componentAppliedLinearAccClamp=<3000 3000 3000>
    componentLinearAccScale=<1 1 1>
    componentLinearVelScale=<1 1 1>
    overrideWorldGravity {z=-980}
    simSpaceSettings {
     maxAngularAcceleration=500
     maxAngularVelocity=3
     maxLinearAcceleration=500
     maxLinearVelocity=500
     velocityScaleZ=0
     worldAlpha=1
     externalLinearDragV=<0.2 0.2>
    }
   }
   n10 = Component To Local(ComponentPose=@n9)
   n11 = Get not Wearing Mask() -> (not Wearing Mask:bool)
   n12 = Anim Dynamics - Bone: dyn_DEER_leaf_1(
    ComponentPose=Local To Component(LocalPose=Blend Poses by bool(BlendPose_0=@n7, BlendPose_1=@n10, bActiveValue=@n11) {blendType=Linear})
    Alpha=Get wearing deer Mask() -> (wearing deer Mask:double)
    ExternalForce="0,0,0"
   ) {
    angularDampingOverride=0.1
    angularSpringConstant=2000
    bAngularSpring=true
    bGravityOverrideInSimSpace=true
    bOverrideAngularDamping=true
    bUseGravityOverride=true
    numSolverIterationsPreUpdate=2
    simulationSpace=BoneRelative
    alphaBoolBlend {blendOption=Linear}
    boundBone {boneName=dyn_DEER_leaf_1}
    componentAppliedLinearAccClamp=<8000 8000 8000>
    componentLinearAccScale=<5 5 5>
    componentLinearVelScale=<5 5 5>
    gravityOverride {z=-980}
    physicsBodyDefinitions [{
      sphereCollisionRadius=0
      boundBone {boneName=dyn_DEER_leaf_1}
      boxExtents=<4 5 5>
      constraintSetup {angularTargetAxis=AxisY, twistAxis=AxisY, angularLimitsMax=<0 30 0>, angularLimitsMin=<0 -30 0>, angularTarget=<0 0 0>}
      localJointOffset=<1 0 0>
     }]
    relativeSpaceBone {boneName=head}
   }
   n13 = "Transform (Modify) Bone - Bone: dyn_DEER_main_1"(
    ComponentPose=@n12
    Alpha=Get wearing Deer Mask ADS() -> (wearing Deer Mask ADS:double)
    Translation="0,2,0"
    Rotation="0,0,0"
    Scale="0.95,0.95,0.95"
   ) {
    scaleMode=BMM_Replace
    scaleSpace=BCS_BoneSpace
    translationMode=BMM_Additive
    alphaBoolBlend {blendOption=Linear}
    boneToModify {boneName=dyn_DEER_main_1}
    scale=<0.95 0.95 0.95>
    translation {y=2}
   }
   n14 = Get wearingBackpack() -> (wearingBackpack:double)
   n15 = "Constraint - dyn_hair_01 (PP)"!(ComponentPose=@n13, Alpha=@n14, ConstraintWeights_0=0.7, ConstraintWeights_1:float=0.3) -> (Pose) {
    alphaBoolBlend {blendOption=Linear}
    boneToModify {boneName=dyn_hair_01}
    constraintSetup [{targetBone {boneName=head}}, {targetBone {boneName=spine_05}}]
    constraintWeights [1,1]
   }
   n16 = "Constraint - dyn_hair_02 (PP)"!(ComponentPose=@n15, Alpha=@n14, ConstraintWeights_0=0.2, ConstraintWeights_1:float=0.8) -> (Pose) {
    alphaBoolBlend {blendOption=Linear}
    boneToModify {boneName=dyn_hair_02}
    constraintSetup [{targetBone {boneName=head}}, {targetBone {boneName=spine_05}}]
    constraintWeights [1,1]
   }
   n17 = "Constraint - dyn_hair_03 (PPP)"!(ComponentPose=@n16, Alpha=@n14, ConstraintWeights_0=0.2, ConstraintWeights_1:float=0.6, ConstraintWeights_2:float=0.2) -> (Pose) {
    alphaBoolBlend {blendOption=Linear}
    boneToModify {boneName=dyn_hair_03}
    constraintSetup [{targetBone {boneName=spine_04}}, {targetBone {boneName=spine_05}}, {targetBone {boneName=head}}]
    constraintWeights [1,1,1]
   }
   n18 = "Constraint - dyn_hair_04 (PPP)"!(ComponentPose=@n17, Alpha=@n14, ConstraintWeights_0=0.5, ConstraintWeights_1:float=0.3, ConstraintWeights_2:float=0.2) -> (Pose) {
    alphaBoolBlend {blendOption=Linear}
    boneToModify {boneName=dyn_hair_04}
    constraintSetup [{targetBone {boneName=spine_04}}, {targetBone {boneName=spine_05}}, {targetBone {boneName=spine_03}}]
    constraintWeights [1,1,1]
   }
   n19 = Get wearingBackpack()
   n20 = "Transform (Modify) Bone - Bone: dyn_hair_01"(ComponentPose=@n18, Alpha=@n19, Translation="0,0,0", Rotation="0,0,0", Scale="1,1,1") {
    rotationMode=BMM_Additive
    rotationSpace=BCS_BoneSpace
    translationMode=BMM_Additive
    translationSpace=BCS_BoneSpace
    alphaBoolBlend {blendOption=Linear}
    alphaScaleBiasClamp {inRange {max=10, min=-30}, outRange {max=0, min=1}}
    boneToModify {boneName=dyn_hair_01}
   }
   n21 = "Transform (Modify) Bone - Bone: dyn_hair_02"(ComponentPose=@n20, Alpha=@n19, Translation="0,0,1", Rotation="-5,0,0", Scale="1,1,1") {
    rotationMode=BMM_Additive
    rotationSpace=BCS_BoneSpace
    translationMode=BMM_Additive
    translationSpace=BCS_BoneSpace
    alphaBoolBlend {blendOption=Linear}
    alphaScaleBiasClamp {inRange {max=40}}
    boneToModify {boneName=dyn_hair_02}
    rotation {pitch=-5}
    translation {z=1}
   }
   n22 = "Transform (Modify) Bone - Bone: dyn_hair_03"(ComponentPose=@n21, Alpha=@n19, Translation="0,0,-2", Rotation="-20,0,0", Scale="1,1,1") {
    rotationMode=BMM_Additive
    rotationSpace=BCS_BoneSpace
    translationMode=BMM_Additive
    translationSpace=BCS_BoneSpace
    alphaBoolBlend {blendOption=Linear}
    alphaScaleBiasClamp {inRange {max=40}}
    boneToModify {boneName=dyn_hair_03}
    rotation {pitch=-20}
    translation {z=-2}
   }
   n23 = Get backBling Head Rot Z() -> (backBling Head Rot Z:double)
   n24 = "Transform (Modify) Bone - Bone: dyn_hair_01"(ComponentPose=@n22, Alpha=@n23, Translation="0,0,3", Rotation="20,0,0", Scale="1,1,1") {
    rotationMode=BMM_Additive
    rotationSpace=BCS_BoneSpace
    translationMode=BMM_Additive
    translationSpace=BCS_BoneSpace
    alphaBoolBlend {blendOption=Linear}
    alphaScaleBiasClamp {bInterpResult=true, bMapRange=true, interpSpeedDecreasing=8, interpSpeedIncreasing=8, inRange {max=0, min=-30}, outRange {max=0, min=1}}
    boneToModify {boneName=dyn_hair_01}
    rotation {pitch=20}
    translation {z=3}
   }
   n25 = "Transform (Modify) Bone - Bone: dyn_hair_01"(ComponentPose=@n24, Alpha=@n23, Translation="0,0,-2", Rotation="0,0,0", Scale="1,1,1") {
    rotationMode=BMM_Additive
    rotationSpace=BCS_BoneSpace
    translationMode=BMM_Additive
    translationSpace=BCS_BoneSpace
    alphaBoolBlend {blendOption=Linear}
    alphaScaleBiasClamp {bInterpResult=true, bMapRange=true, inRange {max=0, min=-30}, outRange {max=0, min=1}}
    boneToModify {boneName=dyn_hair_01}
    translation {z=-2}
   }
   n26 = "Transform (Modify) Bone - Bone: dyn_hair_02"(ComponentPose=@n25, Alpha=@n23, Translation="0,0,-5", Rotation="0,0,0", Scale="1,1,1") {
    rotationMode=BMM_Additive
    rotationSpace=BCS_BoneSpace
    translationMode=BMM_Additive
    translationSpace=BCS_BoneSpace
    alphaBoolBlend {blendOption=Linear}
    alphaScaleBiasClamp {bInterpResult=true, bMapRange=true, inRange {max=0, min=-30}, outRange {max=0, min=1}}
    boneToModify {boneName=dyn_hair_02}
    translation {z=-5}
   }
   n27 = "Transform (Modify) Bone - Bone: dyn_hair_03"(ComponentPose=@n26, Alpha=@n23, Translation="0,0,0", Rotation="-20,0,0", Scale="1,1,1") {
    rotationMode=BMM_Additive
    rotationSpace=BCS_BoneSpace
    translationMode=BMM_Additive
    translationSpace=BCS_BoneSpace
    alphaBoolBlend {blendOption=Linear}
    alphaScaleBiasClamp {bInterpResult=true, bMapRange=true, inRange {max=0, min=-30}, outRange {max=0, min=1}}
    boneToModify {boneName=dyn_hair_03}
    rotation {pitch=-20}
   }
   n28 = Get backBling Head Rot Z()
   n29 = "Transform (Modify) Bone - Bone: dyn_hair_01"(ComponentPose=@n27, Alpha=@n28, Translation="0,0,0", Rotation="-30,0,0", Scale="1,1,1") {
    rotationMode=BMM_Additive
    rotationSpace=BCS_BoneSpace
    translationMode=BMM_Additive
    translationSpace=BCS_BoneSpace
    alphaBoolBlend {blendOption=Linear}
    alphaScaleBiasClamp {bInterpResult=true, bMapRange=true, inRange {max=30}}
    boneToModify {boneName=dyn_hair_01}
    rotation {pitch=-30}
   }
   n30 = "Transform (Modify) Bone - Bone: dyn_hair_02"(ComponentPose=@n29, Alpha=@n28, Translation="0,0,5", Rotation="0,0,0", Scale="1,1,1") {
    rotationMode=BMM_Additive
    rotationSpace=BCS_BoneSpace
    translationMode=BMM_Additive
    translationSpace=BCS_BoneSpace
    alphaBoolBlend {blendOption=Linear}
    alphaScaleBiasClamp {bInterpResult=true, bMapRange=true, inRange {max=30}}
    boneToModify {boneName=dyn_hair_02}
    translation {z=5}
   }
   n31 = "Transform (Modify) Bone - Bone: dyn_hair_03"(ComponentPose=@n30, Alpha=@n28, Translation="0,0,10", Rotation="20,0,0", Scale="1,1,1") {
    rotationMode=BMM_Additive
    rotationSpace=BCS_BoneSpace
    translationMode=BMM_Additive
    translationSpace=BCS_BoneSpace
    alphaBoolBlend {blendOption=Linear}
    alphaScaleBiasClamp {bInterpResult=true, bMapRange=true, inRange {max=30}}
    boneToModify {boneName=dyn_hair_03}
    rotation {pitch=20}
    translation {z=10}
   }
   n32 = RigidBody!(ComponentPose=@n31, Alpha=Get wearingBackpack(), OverrideWorldGravity="0,0,-980", ExternalForce) -> (Pose) {
    bClampLinearTranslationLimitToRefPose=true
    bOverrideWorldGravity=true
    overridePhysicsAsset=&/F_MED_Island_Nomad_FaceAcc_BackBling_Physics
    alphaBoolBlend {blendOption=Linear}
    baseBoneRef {boneName=head}
    componentAppliedLinearAccClamp=<1000 1000 1000>
    componentLinearAccScale=<0.01 0.01 0.01>
    componentLinearVelScale=<0.01 0.01 0.01>
    overrideWorldGravity {z=-980}
   }
   Output Pose(Result=Component To Local(ComponentPose=@n32)) {layerGroup=None}
   n33 = RigidBody(
    ComponentPose=Local To Component(LocalPose=Copy Pose From Mesh() {bUseAttachedParent=true})
    Alpha=Get noBackpack() -> (noBackpack:double)
    OverrideWorldGravity=Get Gravity Override()
   ) {
    bClampLinearTranslationLimitToRefPose=true
    bOverrideWorldGravity=true
    cachedBoundsScale=1.1
    overridePhysicsAsset=&/F_MED_Island_Nomad_FaceAcc_Physics
    simulationSpace=BaseBoneSpace
    alphaBoolBlend {blendOption=Linear}
    baseBoneRef {boneName=head}
    componentAppliedLinearAccClamp=<1000 1000 1000>
    componentLinearAccScale=<0.005 0.005 0.4>
    componentLinearVelScale=<0.005 0.005 0.4>
    overrideWorldGravity {z=-980}
   }
   Save cached pose 'defaultPose'(Pose=Component To Local(ComponentPose=@n33)) {cacheName=defaultPose}
   Get wearing Mask() -> (wearing Mask:double)
   Get wearing Mask()
  ]}
 EventGraph {nodes [
   n1 = Event BlueprintUpdateAnimation()
   > Set Delta Time X(Delta Time X:double=@n1.DeltaTimeX) -> (Output_Get:double)
   > n2 = Sequence!(execute) -> (then_0, then_1, then_2:Exec, then_3:Exec, then_4:Exec, then_5:Exec, then_6:Exec, then_7:Exec, then_8:Exec, then_9:Exec, then_10:Exec)
   > n3 = Gravity Override State Function DEPRECATED(CharacterPartAnimInstance=Self-Reference(), GameTimeInSeconds=GetGameTimeInSeconds(), GravityOverrideParams=Get Gravity Override Params())
   > Set Gravity Override(Gravity Override=@n3)
   n4 = CheckForBackbling(self=Self-Reference())
   n5 = OR Boolean(A=@n4, B=@n4.bIsBackblingACape)
   Set wearing backpack Bool(execute=@n2.then_1, wearing backpack Bool:bool=@n5) -> (Output_Get:bool)
   > n6 = Set wearingBackpack(wearingBackpack:double="To Float (Boolean)"(InBool=@n5)) -> (Output_Get:double)
   > Set noBackpack(noBackpack:double="To Float (Boolean)"(InBool="Not Equal (Float)"(A=@n6, B=1))) -> (Output_Get:double)
   n7 = Cast To FortPlayerPawn(execute=@n26.then, Object=GetOwningActor())
   > n8 = GetVariantChannelActiveIndexForItemActor!(
    execute
    Item:Athena Cosmetic Item Definition Object Reference=/BRCosmetics/Athena/Items/Cosmetics/Characters/CID_A_291_Athena_Commando_F_IslandNomad
    Actor:Actor Object Reference=@n7
    Channel:Gameplay Tag Structure=(TagName="Cosmetics.Variant.Channel.Parts")
    VariantList:Array of Gameplay Tag Structures=Make Array!(
     "[0]"=(TagName="Cosmetics.Variant.Property.Stage1")
     "[1]":Gameplay Tag Structure=(TagName="Cosmetics.Variant.Property.Stage2")
     "[2]":Gameplay Tag Structure=(TagName="Cosmetics.Variant.Property.Stage3")
     "[3]":Gameplay Tag Structure=(TagName="Cosmetics.Variant.Property.Stage4")
     "[4]":Gameplay Tag Structure=(TagName="Cosmetics.Variant.Property.Stage5")
     "[5]":Gameplay Tag Structure=(TagName="Cosmetics.Variant.Property.Stage17")
     "[6]":Gameplay Tag Structure=(TagName="Cosmetics.Variant.Property.Stage18")
     "[7]":Gameplay Tag Structure=(TagName="Cosmetics.Variant.Property.Stage19")
     "[8]":Gameplay Tag Structure=(TagName="Cosmetics.Variant.Property.Stage20")
     "[9]":Gameplay Tag Structure
     "[10]":Gameplay Tag Structure
     "[11]":Gameplay Tag Structure
    ) -> (Array)
    bErrorIfChannelMissing:bool=false
    bErrorIfTagNotInArray:bool=false
   ) -> (then, Index:int=0, bIndexIsValid:bool, bFoundChannel:bool)
   > n9 = Set mask index(mask index:int=@n8) -> (Output_Get:int=0)
   n10 = Branch(execute=@n2.then_10, Condition=OR Boolean(A="Equal (Integer)"(A=@n9), B="Equal (Integer)"(A=@n9, B=1)))
   > Set wearing Mask(wearing Mask:double=0) -> (Output_Get:double)
   > Set not Wearing Mask(not Wearing Mask:bool=true) -> (Output_Get:bool)
   Set wearing Mask(execute=@n10.else, wearing Mask=1)
   > Set not Wearing Mask(not Wearing Mask=false)
   n11 = Set head Rot Z(
    execute=@n2.then_2
    head Rot Z:double=GetAlphaFromReferenceTransformDelta(
     SkeletalMeshComponent=Get body component() -> (body component:Skeletal Mesh Component Object Reference)
     BoneName=head
     BaseName=spine_05
     InputRange=(X=0,Y=30)
     OutputRange=(X=0,Y=30)
     TwistAxis="1,0,0"
    ).RawOutput
   ) -> (Output_Get:double)
   > n12 = Branch(Condition=@n5)
   > Set backBling Head Rot Z(backBling Head Rot Z:double=@n11) -> (Output_Get:double)
   n13 = Branch(execute=@n12.else, Condition=Get bRigidBodyEnabled())
   > Set backBling Head Rot Z(backBling Head Rot Z=0)
   Set backBling Head Rot Z(execute=@n13.else, backBling Head Rot Z=@n11)
   Set body component(execute=@n2.then_3, body component:Skeletal Mesh Component Object Reference=GetOwnerSkeletalMeshComponent()) -> (Output_Get:Skeletal Mesh Component Object Reference)
   n14 = NOT Boolean(A=@n5)
   n15 = Get bIsSkyDiving()
   Set skydive float(execute=@n2.then_5, skydive float:double="To Float (Boolean)"(InBool=AND Boolean(A=@n15, B=@n14))) -> (Output_Get:double)
   > n16 = Branch(Condition=Get bRigidBodyEnabled())
   > Set skyDive Fix Alpha(skyDive Fix Alpha:double=1) -> (Output_Get:double)
   Set skyDive Fix Alpha(execute=@n16.else, skyDive Fix Alpha=0)
   n17 = GetOwnerDirection!(self:Custom Character Part Anim Instance Object Reference) -> (ReturnValue:float)
   n18 = Set RelaxSpeed(
    execute=@n2.then_7
    RelaxSpeed:double=MapRangeClamped(Value="Absolute (Float)"(A=@n17), InRangeA=85, InRangeB=110, OutRangeA=1.5, OutRangeB=4)
   ) -> (Output_Get:double)
   > n19 = Branch(Condition=Get wearing backpack Bool() -> (wearing backpack Bool:bool))
   > Set RelaxSpeed(RelaxSpeed=4)
   Set RelaxSpeed(execute=@n19.else, RelaxSpeed=@n18)
   n20 = Branch(execute=@n2.then_8, Condition=Get bIsSkyDiving())
   > n21 = Branch(Condition=Get bIsParachuteOpened())
   > Set parachute_timer(parachute_timer:double=float + float(A=Get parachute_timer() -> (parachute_timer:double), B=1) {is=title K2Node_CommutativeAssociativeBinaryOperator}) -> (Output_Get:double)
   > n22 = Branch(Condition="float > float"(A=Get parachute_timer(), B=150) {is=title K2Node_CallFunction})
   > Set TrailAlpha(TrailAlpha:double=FInterpTo(Current=Get TrailAlpha() -> (TrailAlpha:double), Target=0.2, DeltaTime=Get Delta Time X() -> (Delta Time X:double), InterpSpeed=1)) -> (Output_Get:double)
   Set TrailAlpha(execute=@n22.else, TrailAlpha=FInterpTo(Current=Get TrailAlpha(), Target=0, DeltaTime=Get Delta Time X(), InterpSpeed=5))
   Set TrailAlpha(execute=[@n20.else, @n21.else], TrailAlpha=0.25)
   > Set parachute_timer(parachute_timer=0)
   > n23 = Branch(Condition=OR Boolean(A=Get bIsJumping(), B=Get bIsFalling()))
   > Set TrailAlpha(TrailAlpha=0.75)
   Set TrailAlpha(execute=@n23.else, TrailAlpha=1)
   n24 = Branch(
    execute=@n2.then_9
    Condition=OR Boolean!(A="Equal (Integer)"(A=@n9, B=6), B="Equal (Integer)"(A=@n9, B=7), C:bool="Equal (Integer)"(A=@n9, B=8)) -> (ReturnValue)
   )
   > Set wearing deer Mask(wearing deer Mask:double=1) -> (Output_Get:double)
   > n25 = Branch(Condition=Get bIsTargeting())
   > Set wearing Deer Mask ADS(wearing Deer Mask ADS:double=1) -> (Output_Get:double)
   Set wearing Deer Mask ADS(execute=@n25.else, wearing Deer Mask ADS=0)
   Set wearing deer Mask(execute=@n24.else, wearing deer Mask=0)
   n26 = Event OnCharacterCustomizationLoaded()
   Event OnBeginJumping()
   > n27 = Sequence!(execute) -> (then_0, then_1)
   > Set jumping Acc Scale Values(jumping Acc Scale Values:Vector="0.5,0.5,0.6") -> (Output_Get:Vector)
   Set jumping Acc values Braid(execute=@n27.then_1, jumping Acc values Braid:Vector="0.01,0.01,0.2") -> (Output_Get:Vector)
   Event OnEndJumping()
   > Delay(Duration=5)
   > n28 = Sequence()
   > Set jumping Acc values Braid(jumping Acc values Braid="0.001,0.001,0.001")
   Set jumping Acc Scale Values(execute=@n28.then_1, jumping Acc Scale Values="0.01,0.01,0.1")
  ]}
]
variables [
 acc values X {type=number}
 acc values Y {type=number}
 acc values Z {type=number}
 As Fort Player Pawn {type=FortPlayerPawn}
 As Fort Player Pawn_0 {type=FortPlayerPawn}
 backBling Head Rot Y {type=boolean}
 backBling Head Rot Z {type=number}
 body component {type=SkeletalMeshComponent}
 Delta Time X {type=number}
 Gravity Override {type=Vector}
 Gravity Override Params {
  type=GravityOverrideParamsStruct
  default {skydive_FloatingDown {gravityOverride=<0 -50 50>, windAmplitude=<200 -250 200>, windFrequency=<0.2 3 4>, windOffset=<0 -250 100>}}
 }
 hat index {type=integer}
 head Rot Y {type=number}
 head Rot Z {type=number}
 is bird {type=boolean}
 is cat {type=boolean}
 Is Jumping {type=number}
 Is Sky Diving {type=boolean}
 jumping Acc Scale Values {type=Vector}
 jumping Acc values Braid {type=Vector}
 Loadout {
  type=FortAthenaLoadout
  default {
   aura=null
   backpack=null
   bannerColorId=""
   bannerIconId=""
   bass=null
   character=null
   drums=null
   glider=null
   guitar=null
   keytar=null
   loadingScreen=null
   mic=null
   mimosa=null
   musicPack=null
   pickaxe=null
   shoes=null
   skyDiveContrail=null
   characterVariantChannels []
   dances []
   itemWraps []
  }
 }
 Loadout_0 {
  type=FortAthenaLoadout
  default {
   aura=null
   backpack=null
   bannerColorId=""
   bannerIconId=""
   bass=null
   character=null
   drums=null
   glider=null
   guitar=null
   keytar=null
   loadingScreen=null
   mic=null
   mimosa=null
   musicPack=null
   pickaxe=null
   shoes=null
   skyDiveContrail=null
   characterVariantChannels []
   dances []
   itemWraps []
  }
 }
 mask index {type=integer}
 maskIsOn {type=boolean}
 NewVar_2 {type=number}
 noBackpack {type=number}
 not Wearing Mask {type=boolean}
 parachute_timer {type=number}
 RelaxSpeed {type=number}
 Skeletal Mesh Component {type=SkinnedMeshComponent, default=null}
 Skeletal Mesh Component_0 {type=SkinnedMeshComponent, default=null}
 skyDive Fix Alpha {type=number}
 skydive float {type=number}
 Tag Container {type=GameplayTagContainer, default {gameplayTags [{tagName=Abilities.Outlander}, {tagName=Cosmetics.Variant.Property.Stage3}]}}
 Target {default=None}
 TrailAlpha {type=number}
 wearing backpack Bool {type=boolean}
 wearing deer Mask {type=number}
 wearing Deer Mask ADS {type=number}
 wearing hat {type=number}
 wearing Mask {type=number}
 wearingBackpack {type=number}
]
